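# Translucent, unlit, two-sided mesh-particle material for the soft energy glow on the Tesla pickaxe's vacuum tube effect.
# Color hue-shifts particle color with panning beam and flame textures, keeps it unshifted mid-U, then boosts it by Fresnel.
# Opacity multiplies particle alpha, view-dependent Fresnel, smoothstepped scrolling flame and energy-flow noise, vertex red and depth fade.
# A RandTile dynamic parameter varies texture tiling per particle.
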Material /Game/Weapons/FORT_Melee/Effects/VacuumTube/M_Tesla_PickAxe_SoftGlow
graph {
 properties {
  bUsedWithMeshParticles=true
  blendMode=BLEND_Translucent
  pixelDepthOffsetMode=PDOM_Legacy
  shadingModel=MSM_Unlit
  twoSided=true
  usedShadingModels=Unlit
 }
 nodes [
  n1 = VectorParameter() {parameterName=TileTex01, defaultValue=<rgba 0.5 2 0 1>}
  n2 = DynamicParameter() {paramNames [RandTile,Param2,Param3,Param4]}
  n3 = Multiply(A=TextureCoordinate() {uTiling=2}, B=Add(A=ComponentMask(None=@n1.RGB) {g=true, r=true}, B=@n2.RandTile)) {constB=4}
  n4 = VectorParameter() {parameterName=TileTex02, defaultValue=<rgba 0.25 1 0 1>}
  n5 = Multiply(A=TextureCoordinate() {uTiling=0.5}, B=Add(A=@n2.RandTile, B=ComponentMask(None=@n4.RGB) {g=true, r=true})) {constB=1.7}
  n6 = Multiply(
   A=TextureSample(UVs=Panner(Coordinate=@n3) {bFractionalPart=true, speedX=1}) {samplerType=SAMPLERTYPE_LinearColor, texture=/Game/Effects/Fort_Effects/Textures/Tile/T_Raygun_Beam_02}.G
   B=TextureSample(UVs=Panner(Coordinate=@n5) {bFractionalPart=true, speedX=1}) {texture=/Game/Effects/Fort_Effects/Textures/Tile/T_Tile_Texture_Cartoon_Flames}.R
  )
  n7 = ScalarParameter() {defaultValue=0.1, parameterName=HueShift_02}
  n8 = ScalarParameter() {parameterName=HueShift}
  n9 = ParticleColor()
  n10 = ComponentMask(None=TextureCoordinate()) {r=true}
  n11 = Multiply(A=Multiply(A=@n10, B=OneMinus(None=@n10))) {constB=4}
  n12 = LinearInterpolate(
   A=MaterialFunctionCall(
    Hue Shift Percentage=Multiply(A=@n6, B=@n7)
    Texture=MaterialFunctionCall(Hue Shift Percentage=@n8, Texture=@n9.RGB) {materialFunction=&HueShift}.Result
   ) {materialFunction=&HueShift}.Result
   B=@n9.RGB
   Alpha=Clamp(None=Power(Base=@n11) {constExponent=8})
  )
  n13 = ScalarParameter() {defaultValue=2, parameterName=FresnelColorMax_Mult}
  n14 = ScalarParameter() {defaultValue=2, parameterName=FresnelColorMax_Expo}
  MaterialFunctionCall(Emissive=LinearInterpolate(
    A=@n12
    B=Multiply(A=@n13, B=@n12)
    Alpha=MaterialFunctionCall(Exponent=@n14) {materialFunction=&MF_Fresnel_FX}.Fresnel
   )) {materialFunction=&MF_FX_EmissiveAdjustments}
  n15 = Constant() {r=1}
  n16 = Constant() {r=8}
  n17 = MaterialFunctionCall(Float3=Transform(None=CameraVectorWS()) {transformSourceType=TRANSFORMSOURCE_World, transformType=TRANSFORM_Local}) {materialFunction=&BreakOutFloat3Components}
  n18 = Power(Base=DotProduct(
    A=MaterialFunctionCall(X=@n17.R, Y=@n17.R) {materialFunction=&MakeFloat3}.Result
    B=CameraVectorWS()
   ))
  n19 = Multiply(
   A=ParticleColor().A
   B=MaterialFunctionCall(Exponent=LinearInterpolate(A=@n15, B=@n16, Alpha=@n18)) {materialFunction=&MF_Fresnel_FX}.Fresnel
  )
  n20 = TextureSample(UVs=Panner(Coordinate=@n3) {speedX=1.6, speedY=-1}) {texture=&T_Tile_Texture_Cartoon_Flames}
  n21 = Multiply(
   A=TextureSample(UVs=Panner(Coordinate=@n5) {speedX=1, speedY=2}) {texture=/Game/Effects/Fort_Effects/Textures/Energy/T_EnergyFlow_Packed}.R
   B=TextureSample(UVs=Panner(Coordinate=@n5) {speedX=0.5, speedY=1}) {texture=&T_EnergyFlow_Packed}.G
  )
  n22 = MaterialFunctionCall(
   Alpha=Power(Base=Multiply(A=@n20.G, B=@n21)) {constExponent=1}
   Min=ScalarParameter() {parameterName=Contrast_01}
   Max=ScalarParameter() {defaultValue=1, parameterName=Contrast_02}
  ) {materialFunction=&SmoothStep}
  Multiply(A=Multiply(A=Multiply(A=@n19, B=@n22.Result), B=VertexColor().R), B=DepthFade(FadeDistance=Constant() {r=12}))
 ]
}
